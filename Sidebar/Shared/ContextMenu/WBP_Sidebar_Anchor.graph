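# Sidebar widget that anchors and toggles a context menu, derived from UIKitContextMenuAnchor.
# On spawn, it passes the view model, interaction context, modality, width override and telemetry info to the menu, optionally binding item clicks to close it.
# In Campaign, it also registers a Back input action that closes the menu.
# It can expose context actions as quick hotkeys, and defers its mouse-driven closed event until the pointer leaves.
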WidgetBlueprint /Sidebar/Shared/ContextMenu/WBP_Sidebar_Anchor
graphs [
 AddInfoForContextMenuTelemetry {nodes [
   n1 = AddInfoForContextMenuTelemetry!() -> (then, InfoKey:String, InfoValue:String) {is=title K2Node_FunctionEntry}
   > Add!(
    execute
    TargetMap:Map of Strings to Strings=Get OptionalInfoForTelemetry() -> (OptionalInfoForTelemetry:Map of Strings to Strings)
    Key:"String (by ref)"=@n1
    Value:"String (by ref)"=@n1.InfoValue
   ) -> (then) {is=title K2Node_CallFunction}
  ]}
 CloseMenu {nodes [
   CloseMenu() {is=title K2Node_FunctionEntry}
   > Close(self:Menu Anchor Object Reference=Get Anchor() -> (Anchor:Menu Anchor Object Reference))
   n1 = Get SpawnedMenu() -> (SpawnedMenu:WBP Sidebar Context Menu Object Reference)
   > n2 = Is Valid(InputObject=@n1)
   Branch(execute=[@n3.then, @n2.Is Not Valid], Condition=IsInSubGame!(SubGame:ESubGame Enum=Campaign) -> (ReturnValue:bool))
   > UnregisterInputAction!(
    execute
    self:Fort Global UIContext Object Reference=Get FortGlobalUIContext() -> (ReturnValue:Fort Global UIContext Object Reference) {is=title K2Node_GetSubsystem}
    UserWidget:User Widget Object Reference=Self-Reference()
    InputActionRow:Data Table Row Handle Structure=Get BackActionHandle() -> (BackActionHandle:Data Table Row Handle Structure)
   ) -> (then)
   n3 = Unbind all Events from On Item Clicked!(
    execute:Exec=@n2.Is Valid
    self:List View Object Reference=Get ListView!(self:WBP Sidebar Context Menu Object Reference=@n1) -> (ListView:Common List View Object Reference)
   ) -> (then:Exec) {is=title K2Node_ClearDelegate}
  ]}
 EventGraph {nodes [
   n1 = "On Menu Open Changed (Anchor)"!() -> (OutputDelegate, then, bIsOpen:bool) {is=title K2Node_ComponentBoundEvent}
   > n2 = Branch(Condition=@n1.bIsOpen)
   > Call On Context Menu Opened!(execute:Exec, self:WBP Sidebar Anchor Object Reference) -> (then:Exec) {is=title K2Node_CallDelegate}
   n3 = Switch on ECommonInputType!(execute=@n2.else, Selection:ECommonInputType Enum=GetCurrentInputType(self=Get CommonInputSubsystem())) -> (MouseAndKeyboard:Exec, Gamepad:Exec, Touch:Exec, Count:Exec) {is=title K2Node_SwitchEnum}
   > n4 = Branch(Condition=IsHovered())
   Call On Context Menu Closed!(execute:Exec=[@n4.else, @n3.Count, @n3.Gamepad, @n3.Touch], self:WBP Sidebar Anchor Object Reference) -> (then:Exec) {is=title K2Node_CallDelegate}
   ToggleContextMenu() {is=title K2Node_CustomEvent}
   > ToggleOpen!(execute, self:Menu Anchor Object Reference=Get Anchor(), bFocusOnOpen:bool=true) -> (then)
   CloseMenuOnOptionClicked!() -> (OutputDelegate, then, Item:Object Reference) {is=title K2Node_CustomEvent}
   > CloseMenu() {is=title K2Node_CallFunction}
   n5 = Event On Menu Spawned!() -> (OutputDelegate, then, Widget:User Widget Object Reference) {is=title K2Node_Event}
   > n6 = Cast To WBP_Sidebar_ContextMenu!(execute, Object=@n5.Widget) -> (then, CastFailed, AsWBP Sidebar Context Menu:WBP Sidebar Context Menu Object Reference) {is=title K2Node_DynamicCast}
   > n7 = SetIsModal!(execute, self:UIKit Context Menu Object Reference=@n6, bInIsModal:bool=Get IsMenuModal() -> (IsMenuModal:bool)) -> (then)
   Init Context Menu!(execute=[@n7.then, @n6.CastFailed], self, Widget:Object Reference=@n5.Widget) -> (then)
   > Branch(Condition=IsInSubGame(SubGame=Campaign))
   > RegisterInputAction!(
    execute
    self:Fort Global UIContext Object Reference=Get FortGlobalUIContext()
    UserWidget:User Widget Object Reference=Self-Reference()
    InputActionRow:Data Table Row Handle Structure=Get BackActionHandle()
    CommitedEvent:Delegate=Create Event()
    InFilterPriority:int=1
   ) -> (then)
   Event PreConstruct()
   > SetPlacement!(
    execute
    self:Menu Anchor Object Reference=Get Anchor()
    InPlacement:EMenuPlacement Enum=Get MenuPlacement() -> (MenuPlacement:EMenuPlacement Enum=MenuPlacement_BelowAnchor)
   ) -> (then)
   > n8 = Branch(Condition=IsValid(Object=Get ViewModel() -> (ViewModel:UIKit Context Menu Viewmodel Object Reference)))
   n9 = Construct Object from Class(execute=@n8.else, Class=/Script/FortniteUI.UIKitContextMenuViewModel) -> (ReturnValue:UIKit Context Menu Viewmodel Object Reference)
   > Set ViewModel(ViewModel:UIKit Context Menu Viewmodel Object Reference=@n9) -> (Output_Get:UIKit Context Menu Viewmodel Object Reference)
   Event OnMouseLeave!() -> (OutputDelegate, then, MouseEvent:"Pointer Event Structure (by ref)") {is=title K2Node_Event}
   > n10 = Branch(Condition=IsOpen!(self:Menu Anchor Object Reference=Get Anchor()) -> (ReturnValue:bool))
   Call On Context Menu Closed(execute=@n10.else)
   Event Destruct()
   Event Construct()
  ]}
 Init Context Menu {nodes [
   n1 = Init Context Menu!() -> (then, Widget:Object Reference) {is=title K2Node_FunctionEntry}
   > n2 = Cast To WBP_Sidebar_ContextMenu(Object=@n1)
   > Set SpawnedMenu(SpawnedMenu:WBP Sidebar Context Menu Object Reference=@n2) -> (Output_Get:WBP Sidebar Context Menu Object Reference)
   n3 = Break S UI Bool Float!(S_UI_BoolFloat:"S UI Bool Float Structure (by ref)"=Get ContextMenu_WidthOverride() -> (ContextMenu_WidthOverride:S UI Bool Float Structure)) -> (Enable_3_#1:bool=~, Value_4_#2:double=~) {is=title K2Node_BreakStruct}
   > n4 = Branch(Condition=@n3)
   n5 = Get SizeBox!(self:WBP Sidebar Context Menu Object Reference=@n2) -> (SizeBox:Size Box Object Reference)
   > n6 = SetWidthOverride(self=@n5, InWidthOverride=@n3.Value_4_#2)
   Set UseBottomBarButtons!(
    execute=[@n6.then, @n9.then]
    UseBottomBarButtons:bool=Get UseBottomBarButtons() -> (UseBottomBarButtons:bool)
    self:WBP Sidebar Context Menu Object Reference=@n2
   ) -> (then, Output_Get:bool)
   > n7 = Sequence()
   n8 = Get SpawnedMenu()
   > SetUIKitContextMenuViewModel!(execute, self:WBP Sidebar Context Menu Object Reference=@n8, Viewmodel:UIKit Context Menu Viewmodel Object Reference=Get ViewModel()) -> (then)
   > SetInteractionContext!(
    execute
    self:UIKit Context Menu Object Reference=@n8
    InContext:UIKit Context Menu Action Context Object Reference=Get InteractionContext() -> (InteractionContext:UIKit Context Menu Action Context Object Reference)
   ) -> (then)
   > SetParentIDForTelemetry!(
    execute
    self:WBP Sidebar Context Menu Object Reference=@n8
    ParentID:String=Get ParentIDForTelemetry() -> (ParentIDForTelemetry:String)
    IsOptionalParentIdEGPI:bool=Get IsParentIdEGPI() -> (IsParentIdEGPI:bool)
   ) -> (then)
   > SetOptionalInfoForTelemetry!(execute, self:WBP Sidebar Context Menu Object Reference=@n8, InfoForTelemetry:Map of Strings to Strings=Get OptionalInfoForTelemetry()) -> (then)
   > Bind Event to Close Menu on Select Delegate(self:WBP Sidebar Context Menu Object Reference=@n8, Delegate=Create Event(self=Get Anchor())) {is=title K2Node_AddDelegate}
   > Call Init Finished!(execute:Exec, self:WBP Sidebar Anchor Object Reference, Object:Object Reference=@n8) -> (then:Exec) {is=title K2Node_CallDelegate}
   Branch(execute=@n7.then_1, Condition=Get CloseMenuOnContextClicked() -> (CloseMenuOnContextClicked:bool))
   > Bind Event to On Item Clicked(self:List View Object Reference=Get ListView(self=Get SpawnedMenu()), Delegate=Create Event()) {is=title K2Node_AddDelegate}
   n9 = ClearWidthOverride(execute=@n4.else, self:Size Box Object Reference=@n5)
  ]}
 InitFinished {nodes [InitFinished!() -> (then, Object:Object Reference) {is=title K2Node_FunctionEntry}]}
 OnBackPressed {nodes [
   OnBackPressed() {is=title K2Node_FunctionEntry}
   > CloseMenu()
   > Return Node!(execute, bPassThrough:bool=false)
  ]}
 OnContextMenuClosed {nodes [OnContextMenuClosed() {is=title K2Node_FunctionEntry}]}
 OnContextMenuOpened {nodes [OnContextMenuOpened() {is=title K2Node_FunctionEntry}]}
 SetContext {nodes [
   n1 = SetContext!() -> (then, InteractionContext_Data:UIKit Context Menu Action Context Object Reference) {is=title K2Node_FunctionEntry}
   > Set InteractionContext(InteractionContext:UIKit Context Menu Action Context Object Reference=@n1) -> (Output_Get:UIKit Context Menu Action Context Object Reference)
   > Update Quick Actions()
  ]}
 SetParentAdditionalDetailsForContextMenuTelemetry {nodes [
   n1 = SetParentAdditionalDetailsForContextMenuTelemetry!() -> (then, OptionalAdditionalDetails:Flow Analytics Additional Info Structure) {is=title K2Node_FunctionEntry}
   > Set ParentAdditionalDetailsForTelemetry(ParentAdditionalDetailsForTelemetry:Flow Analytics Additional Info Structure=@n1) -> (Output_Get:Flow Analytics Additional Info Structure)
  ]}
 SetParentIDForContextMenuTelemetry {nodes [
   n1 = SetParentIDForContextMenuTelemetry!() -> (then, ParentID:String, IsParentIdEGPI:bool=~) {is=title K2Node_FunctionEntry}
   > Set ParentIDForTelemetry(ParentIDForTelemetry:String=@n1) -> (Output_Get:String)
   > Set IsParentIdEGPI(IsParentIdEGPI:bool=@n1.IsParentIdEGPI) -> (Output_Get:bool)
  ]}
 Toggle Quick Actions {nodes [
   n1 = Toggle Quick Actions!() -> (then, Enabled:bool=~) {is=title K2Node_FunctionEntry}
   > Branch(Condition="Not Equal (Boolean)"(A=@n1, B=Get UseQuickHotkeys() -> (UseQuickHotkeys:bool)) -> (ReturnValue=~))
   > Set UseQuickHotkeys(UseQuickHotkeys:bool=@n1) -> (Output_Get:bool)
   > Update Quick Actions()
  ]}
 Update Quick Actions {nodes [
   Update Quick Actions() {is=title K2Node_FunctionEntry}
   > n1 = Branch(Condition=Get UseQuickHotkeys())
   n2 = Get InteractionContext()
   > Is Valid(InputObject=@n2)
   > SetQuickActions!(
    execute
    self:UIKit Context Menu Anchor Object Reference
    QuickActions:Array of UIKit Context Menu Action Info Structures=GetActionInfos!(self:UIKit Context Menu Action Context Object Reference=@n2) -> (ReturnValue:Array of UIKit Context Menu Action Info Structures)
   ) -> (then)
   ClearQuickActions(execute=@n1.else, self:UIKit Context Menu Anchor Object Reference)
  ]}
]
variables [
 BackActionHandle {type=DataTableRowHandle, default {dataTable=&InputActionDataTable, rowName=Back}}
 CloseMenuOnContextClicked {type=boolean}
 ContextMenu_WidthOverride {type=S_UI_BoolFloat, default {enable=false, value=0}}
 InitFinished {}
 InteractionContext {type=UIKitContextMenuActionContext, default=null}
 IsMenuModal {type=boolean, default=true}
 IsParentIdEGPI {type=boolean}
 MenuPlacement {type=EMenuPlacement, default=MenuPlacement_MenuRight}
 OnContextMenuClosed {}
 OnContextMenuOpened {}
 OptionalInfoForTelemetry {type="map<string>", default {}}
 ParentAdditionalDetailsForTelemetry {type=FlowAnalyticsAdditionalInfo, default {}}
 ParentIDForTelemetry {type=string, default=""}
 SpawnedMenu {type=WBP_Sidebar_ContextMenu_C, default=null}
 Triggering Input Action {type=DataTableRowHandle, default {dataTable=null, rowName=None}}
 UseBottomBarButtons {type=boolean}
 UseQuickHotkeys {type=boolean}
 ViewModel {type=UIKitContextMenuViewModel, default=null}
]
widgets {
 info {inheritedWidgetCount=1, parentClass=/Script/FortniteUI.UIKitContextMenuAnchor, rootWidgetClass=&GridPanel, widgetCount=2}
 widgets [
  GridPanel_Root {widgetClassPath=&GridPanel, bIsVariable=true, widget=GridPanel_Root, properties {columnFill [1], rowFill [1]}}
  Anchor {
   widgetClassPath=/Script/UMG.MenuAnchor
   bInherited=true
   bIsVariable=true
   parent=GridPanel_Root
   widget=Anchor
   properties {
    menuClass=/FortUILibrary/Systems/ContextMenu/WBP_UIKit_ContextMenu.WBP_UIKit_ContextMenu_C
    placement=MenuPlacement_MenuRight
    showMenuBackground=false
    visibility=SelfHitTestInvisible
   }
   slot {class=GridSlot, properties {padding=<ltrb -16 8 -16 8>}}
  }
 ]
}
