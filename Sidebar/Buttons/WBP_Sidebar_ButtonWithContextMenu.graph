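# Sidebar settings button, derived from WBP_Sidebar_ButtonGeneric, that opens a context menu of options and shows the chosen one as its second text.
# Options fill an action context on the anchor, from either the local Options map or bindings for a FortSocialSettingDiscreteVM.
# UpdateOptionSelection moves the check icon; clicks fire OnButtonOptionClicked in local mode, otherwise write the view model by index.
# Compact UI adjusts padding, arrow size and menu placement.
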WidgetBlueprint /Sidebar/Buttons/WBP_Sidebar_ButtonWithContextMenu
graphs [
 BindSetActiveOption {nodes [
   n1 = BindSetActiveOption!() -> (then, Text:"Text (by ref)") {is=title K2Node_FunctionEntry}
   > n2 = GetOptionName!(execute, self, Text:"Text (by ref)"=@n1) -> (then, Output:Name=None)
   > n3 = Set NewOption(NewOption:Name=@n2) -> (Output_Get:Name=None)
   n4 = Get ActiveOption() -> (ActiveOption:Name=None)
   > n5 = Branch(Condition="Equal (Name)"(A=@n3, B=@n4))
   UpdateOptionSelection!(execute=@n5.else, self, CurrentSelection:Name=@n4, NewSelection:"Name (by ref)"=Get NewOption() -> (NewOption:Name=None)) -> (then)
  ]}
 BindSetContextOptions {nodes [
   n1 = BindSetContextOptions!() -> (then, Array:Array of Texts) {is=title K2Node_FunctionEntry}
   > Set SelectedOption(SelectedOption:int=GetSettingValueIndex!(self:Fort Social Setting Discrete VM Object Reference=Get FortSocialSettingDiscreteVM() -> (FortSocialSettingDiscreteVM:Fort Social Setting Discrete VM Object Reference)) -> (ReturnValue:int=0)) -> (Output_Get:int=0)
   > RemoveAllInteractions(self:UIKit Context Menu Action Context Object Reference=Get LocalContext() -> (LocalContext:UIKit Context Menu Action Context Object Reference))
   > n2 = For Each Loop(Array:Array of Texts=@n1) -> (Array Element:Text)
   n3 = String To Name(InString="To String (Integer)"(InInt=@n2.Array Index))
   > Add!(
    execute
    TargetMap:Map of Names to Texts=Get Options() -> (Options:Map of Names to Texts)
    Key:"Name (by ref)"=@n3
    Value:"Text (by ref)"=@n2
   ) -> (then) {is=title K2Node_CallFunction}
   > AddInteraction!(
    execute
    self:UIKit Context Menu Action Context Object Reference=Get LocalContext()
    InActionInfo:"UIKit Context Menu Action Info Structure (by ref)"=Make UIKit Context Menu Action Info!(
     DisplayName:Text=@n2
     IconTexture:Texture 2D Soft Object Reference
     IconMaterial:Material Instance Soft Object Reference
     SortPriority:int=1000
     bIsAvailable:bool=True
     bIsEnabled:bool=True
     bIsSelected:bool=False
     bAddDivider:bool=False
     bPositionDividerAfter:bool=False
     bCloseMenuOnSelect:bool=False
     InteractionNameID:Name=@n3
     Interaction:Data Table Row Handle Structure=(DataTable=None,RowName="")
     PressedDelegate:Delegate="(null).None"
     ReleasedDelegate:Delegate="(null).None"
    ) -> (UIKitContextMenuActionInfo:UIKit Context Menu Action Info Structure) {is=title K2Node_MakeStruct}
    InInteractionDelegate:"Delegate (by ref)"=Create Event()
   ) -> (then)
  ]}
 BindSetIsEnabled {nodes [
   n1 = BindSetIsEnabled!() -> (then, bInIsEnabled:bool=~) {is=title K2Node_FunctionEntry}
   > SetIsEnabled(bInIsEnabled=@n1)
  ]}
 ButtonSetup {nodes [
   Inputs()
   > SetText!(execute, self:WBP UIKit Button Generic Base Object Reference, Text:Text=Get OptionName() -> (OptionName:Text)) -> (then)
   n1 = Get OptionDescription() -> (OptionDescription:Text)
   n2 = TextIsEmpty(InText=@n1)
   n3 = Get OptionText() -> (OptionText:Common Text Object Reference)
   > SetVisibility(self=@n3, InVisibility=Select(Option 0=SelfHitTestInvisible, Option 1=Collapsed, Index=@n2))
   > n4 = Branch(Condition=@n2)
   Set Orientation!(
    execute=[@n6.then, @n4.then]
    Orientation:EOrientation Enum=Orient_Horizontal
    self:Stack Box Object Reference=Get Main_Stack() -> (Main_Stack:Stack Box Object Reference)
   ) -> (then, Output_Get:EOrientation Enum=Orient_Horizontal)
   > n5 = Cast To StackBoxSlot!(
    execute
    Object=Get Slot!(self:Widget Object Reference=Get Block_SecondText() -> (Block_SecondText:WBP UIKit Block Text Object Reference)) -> (Slot:Panel Slot Object Reference)
   ) -> (then, CastFailed, AsStack Box Slot:Stack Box Slot Object Reference) {is=title K2Node_DynamicCast}
   > Set Size!(
    execute
    Size:Slate Child Size Structure=Make Slate Child Size!(Value:float=1, SizeRule:ESlateSizeRule Enum=Fill) -> (SlateChildSize:Slate Child Size Structure) {is=title K2Node_MakeStruct}
    self:Stack Box Slot Object Reference=@n5
   ) -> (then, Output_Get:Slate Child Size Structure)
   > Outputs!(then:Exec)
   n6 = "SetText (Text)"(execute=@n4.else, self=@n3, InText=@n1)
  ]}
 ConstructContext {nodes [
   Inputs()
   > n1 = Get!(execute:Exec) -> (then:Exec, else:Exec, LocalContext:UIKit Context Menu Action Context Object Reference)
   Outputs(then=[@n4.then, @n1.then])
   n2 = Construct Object from Class(execute=@n1.else, Class=/Script/FortniteUI.UIKitContextMenuActionContext) -> (ReturnValue:UIKit Context Menu Action Context Object Reference)
   > n3 = Set LocalContext(LocalContext:UIKit Context Menu Action Context Object Reference=@n2) -> (Output_Get:UIKit Context Menu Action Context Object Reference)
   > n4 = SetContext!(
    execute
    self:WBP Sidebar Anchor Object Reference=Get Sidebar_Anchor() -> (Sidebar_Anchor:WBP Sidebar Anchor Object Reference)
    InteractionContext_Data:UIKit Context Menu Action Context Object Reference=@n3
   ) -> (then)
  ]}
 DesignTimePreset {nodes [
   n1 = Inputs!() -> (execute, DesignTime:bool=~)
   > n2 = Branch(Condition=@n1)
   > n3 = SetSecondText!(execute, self:WBP UIKit Button Generic Base Object Reference, SecondText:Text="INVTEXT(\"Value\")") -> (then)
   Outputs(then=[@n3.then, @n2.else])
  ]}
 EventGraph {nodes [
   "On Context Menu Opened (Sidebar_Anchor)"() {is=title K2Node_ComponentBoundEvent}
   > n1 = Branch(Condition=Get UseLocalOptions() -> (UseLocalOptions:bool))
   n2 = GetOptionName(
    execute=@n1.else
    Text=GetSettingValue!(self:Fort Social Setting Discrete VM Object Reference=Get FortSocialSettingDiscreteVM()) -> (ReturnValue:Text)
   ) -> (Output=None)
   > UpdateOptionSelection(CurrentSelection=Get ActiveOption() -> (ActiveOption=None), NewSelection=@n2)
   n3 = OnOptionClicked!() -> (OutputDelegate, then, ButtonID:Name) {is=title K2Node_CustomEvent}
   n4 = Get ActiveOption() -> (ActiveOption=None)
   > n5 = Branch(Condition="Equal (Name)"(A=@n3.ButtonID, B=@n4))
   UpdateOptionSelection(execute=@n5.else, CurrentSelection=@n4, NewSelection=@n3.ButtonID)
   > n6 = Branch(Condition=Get UseLocalOptions())
   > Call On Button Option Clicked!(execute:Exec, self:WBP Sidebar Button With Context Menu Object Reference, Button ID:Name=@n3.ButtonID) -> (then:Exec) {is=title K2Node_CallDelegate}
   SetSettingValueByIndex!(
    execute=@n6.else
    self:Fort Social Setting Discrete VM Object Reference=Get FortSocialSettingDiscreteVM()
    index:int=String To Integer(InString="To String (Name)"(InName=@n3.ButtonID))
   ) -> (then)
   n7 = Event PreConstruct()
   > Parent: PreConstruct(IsDesignTime=@n7.IsDesignTime)
   > Design Time Preset!(execute:Exec, DesignTime:bool=@n7.IsDesignTime) -> (then:Exec) {is=title K2Node_MacroInstance}
   > Construct Context!(execute:Exec) -> (then:Exec) {is=title K2Node_MacroInstance}
   > n8 = Branch(Condition=Get UseLocalOptions())
   > n9 = SetLocalContextOptions()
   Button Setup!(execute:Exec=[@n9.then, @n8.else]) -> (then:Exec) {is=title K2Node_MacroInstance}
   > SetCompactUIStyle(self:WBP Sidebar Button Generic Object Reference) {is=title K2Node_CallFunction}
   Event On Clicked()
   > Parent: On Clicked() {is=title K2Node_CallParentFunction}
   > ToggleContextMenu(self:WBP Sidebar Anchor Object Reference=Get Sidebar_Anchor())
  ]}
 GetOptionName {nodes [
   n1 = GetOptionName!() -> (then, Text:"Text (by ref)") {is=title K2Node_FunctionEntry}
   n2 = Get Options()
   > n3 = Values!(execute, TargetMap:Map of Names to Texts=@n2) -> (then, Values:Array of Texts)
   n4 = Find Item!(TargetArray:Array of Texts=@n3, ItemToFind:"Text (by ref)"=@n1) -> (ReturnValue)
   > n5 = Branch(Condition="integer > integer"(A=@n4, B=-1))
   > n6 = Keys!(execute, TargetMap:Map of Names to Texts=@n2) -> (then, Keys:Array of Names)
   > Return Node!(execute, Output:"Name (by ref)"="Get (a ref)"(Array:Array of Names=@n6, Dimension 1=@n4) -> (Output:"Name (by ref)"))
   PrintString!(execute=@n5.else, InString=Sidebar: NO SUCH OPTION, bPrintToScreen, bPrintToLog, TextColor=(R=0.7753,G=0,B=0,A=1), Duration=6, Key) -> (then)
   > Return Node(Output=None)
  ]}
 OnButtonOptionClicked {nodes [OnButtonOptionClicked!() -> (then, Button ID:Name) {is=title K2Node_FunctionEntry}]}
 SetButtonVM {nodes [
   n1 = SetButtonVM!() -> (then, ViewModel:Fort Social Setting Discrete VM Object Reference) {is=title K2Node_FunctionEntry}
   > Set FortSocialSettingDiscreteVM(FortSocialSettingDiscreteVM:Fort Social Setting Discrete VM Object Reference=@n1) -> (Output_Get:Fort Social Setting Discrete VM Object Reference)
  ]}
 SetCompactUIStyle {nodes [
   SetCompactUIStyle() {is=title K2Node_FunctionEntry}
   > n1 = Branch(Condition=Get Override Compact Style() -> (Override Compact Style:bool))
   Parent: SetCompactUIStyle(execute=@n1.else) {is=title K2Node_CallParentFunction}
   > Branch(Condition=IsUsingCompactUI())
   n2 = Get ArrowIcon() -> (ArrowIcon:Image Object Reference)
   > SetDesiredSizeOverride!(execute, self:Image Object Reference=@n2, DesiredSize:Vector 2D Structure=(X=32,Y=32)) -> (then)
   n3 = SlotAsGridSlot(Widget=@n2)
   n4 = Get!(self:Grid Slot Object Reference=@n3) -> (Padding_Left:float=~, Padding_Top:float=~, Padding_Right:float=~, Padding_Bottom:float=~)
   > SetPadding!(
    execute
    self:Grid Slot Object Reference=@n3
    InPadding_Left:float=@n4
    InPadding_Top:float=24
    InPadding_Right:float=12
    InPadding_Bottom:float=@n4.Padding_Bottom
   ) -> (then)
   > SetGridSlotPadding!(
    execute
    Widget:Widget Object Reference=Get OptionText()
    Padding_Left:float=18
    Padding_Top:float=80
    Padding_Right:float=30
    Padding_Bottom:float=15
   ) -> (then)
   n5 = Get Sidebar_Anchor()
   > SetPlacement!(
    execute
    self:Menu Anchor Object Reference=Get Anchor!(self:UIKit Context Menu Anchor Object Reference=@n5) -> (Anchor:Menu Anchor Object Reference)
    InPlacement:EMenuPlacement Enum=MenuPlacement_AboveRightAnchor
   ) -> (then)
   n6 = SlotAsGridSlot(Widget=@n5)
   n7 = Get(self=@n6) -> (Padding_Left=~, Padding_Top=~, Padding_Right=~, Padding_Bottom=~)
   > SetPadding!(execute, self=@n6, InPadding:Margin Structure=Make Margin(Left=@n7, Top=@n7.Padding_Top, Right=48, Bottom=@n7.Padding_Bottom)) -> (then)
  ]}
 SetLocalActiveOption {nodes [
   n1 = SetLocalActiveOption!() -> (then, Name:Name) {is=title K2Node_FunctionEntry}
   > UpdateOptionSelection(CurrentSelection=Get ActiveOption() -> (ActiveOption=None), NewSelection=@n1)
  ]}
 SetLocalContextOptions {nodes [
   SetLocalContextOptions() {is=title K2Node_FunctionEntry}
   > RemoveAllInteractions(self=Get LocalContext())
   > n1 = "For Each Loop (Map)"!(execute:Exec, MapPin:Map of Names to Texts=Get Options(), BreakPin:Exec) -> (then:Exec, KeyPin:Name, ValuePin:Text, CompletedPin:Exec)
   > AddInteraction(
    self=Get LocalContext()
    InActionInfo=Make UIKit Context Menu Action Info(
     DisplayName=@n1.ValuePin
     SortPriority=1000
     bIsAvailable=True
     bIsEnabled=True
     bIsSelected=False
     bAddDivider=False
     bPositionDividerAfter=False
     bCloseMenuOnSelect=False
     InteractionNameID=@n1
     Interaction=(DataTable=None,RowName="")
     PressedDelegate="(null).None"
     ReleasedDelegate="(null).None"
    )
    InInteractionDelegate=Create Event()
   )
  ]}
 UpdateOptionSelection {nodes [
   UpdateOptionSelection!() -> (then, CurrentSelection:Name, NewSelection:"Name (by ref)") {is=title K2Node_FunctionEntry}
   n1 = Find(TargetMap:Map of Names to Texts=Get Options(), Key:"Name (by ref)"=Get CurrentSelection() -> (CurrentSelection:Name=None)) -> (Value:Text)
   > n2 = Branch(Condition=@n1.ReturnValue)
   > n3 = UpdateInteraction!(
    execute
    self:UIKit Context Menu Action Context Object Reference=Get LocalContext()
    InActionInfo:"UIKit Context Menu Action Info Structure (by ref)"=Make UIKit Context Menu Action Info(
     DisplayName=@n1
     SortPriority=1000
     bIsAvailable=True
     bIsEnabled=True
     bIsSelected=False
     bAddDivider=False
     bPositionDividerAfter=False
     bCloseMenuOnSelect=False
     InteractionNameID=Get CurrentSelection() -> (CurrentSelection=None)
     Interaction=(DataTable=None,RowName="")
     PressedDelegate="(null).None"
     ReleasedDelegate="(null).None"
    )
   ) -> (then)
   n4 = Find(TargetMap=Get Options(), Key=Get NewSelection() -> (NewSelection:"Name (by ref)"=None))
   Branch(execute=[@n3.then, @n2.else], Condition=@n4.ReturnValue)
   > n5 = Set NewOptionText(NewOptionText:Text=@n4) -> (Output_Get:Text)
   > UpdateInteraction(
    self=Get LocalContext()
    InActionInfo=Make UIKit Context Menu Action Info(
     DisplayName=@n5
     IconTexture=Get ActiveIcon() -> (ActiveIcon:Texture 2D Soft Object Reference)
     SortPriority=1000
     bIsAvailable=True
     bIsEnabled=True
     bIsSelected=False
     bAddDivider=False
     bPositionDividerAfter=False
     bCloseMenuOnSelect=False
     InteractionNameID=Get NewSelection() -> (NewSelection=None)
     Interaction=(DataTable=None,RowName="")
     PressedDelegate="(null).None"
     ReleasedDelegate="(null).None"
    )
   )
   > Set ActiveOption(ActiveOption:Name=Get NewSelection() -> (NewSelection=None)) -> (Output_Get:Name=None)
   > SetSecondText(SecondText=Get NewOptionText() -> (NewOptionText:Text))
  ]}
]
variables [
 ActiveIcon {type=Texture2D, default=/FortUILibrary/Textures/Icons/T_UI_Icon_UIKit_Check}
 ActiveOption {type=string, default=None}
 LocalContext {type=UIKitContextMenuActionContext, default=null}
 OnButtonOptionClicked {}
 OptionDescription {type=string, default=""}
 OptionName {type=string, default=Option}
 Options {type="map<string>", default {}}
 Override Compact Style {type=boolean}
 UseLocalOptions {type=boolean}
]
widgets {
 info {
  inheritedWidgetCount=0
  namedSlotCount=1
  parentClass=&/WBP_Sidebar_ButtonGeneric.WBP_Sidebar_ButtonGeneric_C
  rootWidgetClass=null
  widgetCount=4
 }
 widgets [
  Content {widgetClassPath=&GridPanel, widget=Content, properties {columnFill [1], rowFill [1]}}
  Sidebar_Anchor {
   widgetClassPath=/Sidebar/Shared/ContextMenu/WBP_Sidebar_Anchor
   bIsVariable=true
   parent=Content
   widget=Sidebar_Anchor
   properties {closeMenuOnContextClicked=true, menuPlacement=MenuPlacement_MenuLeft}
   slot {class=GridSlot, properties {horizontalAlignment=HAlign_Right, nudge=<16 0>}}
  }
  ArrowIcon {
   widgetClassPath=&Image
   bIsVariable=true
   parent=Content
   widget=ArrowIcon
   properties {brush {
     bIsDynamicallyLoaded=false
     drawAs=Image
     imageType=FullColor
     mirroring=NoMirror
     resourceName=None
     resourceObject=/FortUILibrary/Textures/MaterialInstances/MI_UIKit_Icon_ExpandArrow
     tiling=NoTile
     imageSize=<18 18>
     margin=<ltrb 0 0 0 0>
     outlineSettings {
      bUseBrushTransparency=false
      roundingType=HalfHeightRadius
      width=0
      color {colorUseRule=UseColor_Specified, specifiedColor=<rgba 0 0 0 0>}
      cornerRadii=<0 0 0 0>
     }
     tintColor {colorUseRule=UseColor_Specified, specifiedColor=<rgba 1 1 1 1>}
     uVRegion {bIsValid=false, max=<0 0>, min=<0 0>}
    }}
   slot {class=GridSlot, properties {horizontalAlignment=HAlign_Right, verticalAlignment=VAlign_Top, padding=<ltrb 0 21 18 0>}}
  }
  OptionText {
   widgetClassPath=&CommonTextBlock
   bIsVariable=true
   parent=Content
   widget=OptionText
   properties {
    autoWrapText=true
    mobileFontSizeMultiplier=1.4
    colorAndOpacity {colorUseRule=UseColor_Specified, specifiedColor=<rgba 1 1 1 0.65>}
    font {
     bForceMonospaced=false
     bMaterialIsStencil=false
     fontMaterial=null
     fontObject=&HeadingNow
     letterSpacing=10
     monospacedWidth=1
     size=12.75
     skewAmount=0
     typefaceFontName=74 Regular
     outlineSettings {
      bApplyOutlineToDropShadows=false
      bMiteredCorners=false
      bSeparateFillAlpha=false
      outlineMaterial=null
      outlineSize=0
      outlineColor=<rgba 0 0 0 1>
     }
    }
    shadowOffset=<0 0>
   }
   slot {class=GridSlot, properties {verticalAlignment=VAlign_Top, padding=<ltrb 18 60 30 15>}}
  }
 ]
}
