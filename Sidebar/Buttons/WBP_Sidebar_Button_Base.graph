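# Base sidebar button widget, built on UIKitHoldableModularButton, with a background block and named slots for item content and controls.
# PreConstruct sets backplate color from ButtonType, applies corner radius, and enlarges minimum size for compact UI.
# Input-method changes make the button selectable only for touch.
# Hover events (when not selectable) and selection events (when selectable) share one backplate highlight handler, which does nothing while touch is in use.
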WidgetBlueprint /Sidebar/Buttons/WBP_Sidebar_Button_Base
graphs [
 EventGraph {nodes [
   Event PreConstruct()
   > InitializeButtonVisuals()
   > SetCompactUIStyle() {is=title K2Node_CallFunction}
   Event On Hovered()
   n1 = NOT Boolean(A=Get bSelectable() -> (bSelectable:bool))
   > n2 = Branch(Condition=@n1)
   OnHandleButtonHoverOrUnhovered!(execute=[@n2.then, @n7.then], self, IsHovered:bool=true) -> (then)
   n3 = Event On Input Method Changed!() -> (OutputDelegate, then, CurrentInputType:ECommonInputType Enum) {is=title K2Node_Event}
   > Parent: On Input Method Changed!(execute, CurrentInputType:ECommonInputType Enum=@n3.CurrentInputType) -> (then) {is=title K2Node_CallParentFunction}
   > n4 = Set with Broadcast CurrentInputType!(execute, CurrentInputType:ECommonInputType Enum=@n3.CurrentInputType) -> (then, Output_Get:ECommonInputType Enum)
   > OnHandleInputMethodChanged!(execute, self, NewInputType:ECommonInputType Enum=@n4) -> (then)
   Event On Unhovered()
   > n5 = Branch(Condition=@n1)
   OnHandleButtonHoverOrUnhovered(execute=[@n5.then, @n8.then], IsHovered=false)
   Event On Selected()
   n6 = Get bSelectable()
   > n7 = Branch(Condition=@n6)
   Event On Deselected()
   > n8 = Branch(Condition=@n6)
  ]}
 InitializeButtonVisuals {nodes [
   InitializeButtonVisuals() {is=title K2Node_FunctionEntry}
   > SetBackplateColor!(
    execute
    self:WBP Sidebar Block Backplate Object Reference=Get Backplate!(self:WBP Sidebar Block Background Button Object Reference=Get Background() -> (Background:WBP Sidebar Block Background Button Object Reference)) -> (Backplate:WBP Sidebar Block Backplate Object Reference)
    Backplate Switch:E_UI_Sidebar_Button_Backplate_Color Enum=Select!(
     NewEnumerator0:E_UI_Sidebar_Button_Backplate_Color Enum=NewEnumerator1
     NewEnumerator1:E_UI_Sidebar_Button_Backplate_Color Enum=NewEnumerator1
     NewEnumerator2:E_UI_Sidebar_Button_Backplate_Color Enum=NewEnumerator5
     NewEnumerator3:E_UI_Sidebar_Button_Backplate_Color Enum=NewEnumerator0
     NewEnumerator4:E_UI_Sidebar_Button_Backplate_Color Enum=NewEnumerator0
     NewEnumerator6:E_UI_Sidebar_Button_Backplate_Color Enum=NewEnumerator12
     NewEnumerator7:E_UI_Sidebar_Button_Backplate_Color Enum=NewEnumerator0
     NewEnumerator8:E_UI_Sidebar_Button_Backplate_Color Enum=NewEnumerator12
     Index:E_UI_Sidebar_Button_Type Enum=Get ButtonType() -> (ButtonType:E_UI_Sidebar_Button_Type Enum=NewEnumerator0)
    ) -> (ReturnValue:E_UI_Sidebar_Button_Backplate_Color Enum)
   ) -> (then)
   > SetCornerRadius!(
    execute
    self:WBP Sidebar Block Background Button Object Reference=Get Background()
    Radius In Pixels:bool=Get RadiusInPixels() -> (RadiusInPixels:bool)
    Corner Radius:double=Get CornerRadius() -> (CornerRadius:double)
   ) -> (then)
  ]}
 OnHandleButtonHoverOrUnhovered {nodes [
   OnHandleButtonHoverOrUnhovered!() -> (then, IsHovered:bool=~) {is=title K2Node_FunctionEntry}
   > n1 = Branch(Condition=IsUsingTouch(PlayerController=GetOwningPlayer()))
   n2 = Branch(execute=@n1.else, Condition=Get IsHovered() -> (IsHovered:bool))
   > BackplateChangeOnHover!(
    execute
    self:WBP Sidebar Block Background Button Object Reference=Get Background()
    Button Type:E_UI_Sidebar_Button_Type Enum=Get ButtonType() -> (ButtonType=NewEnumerator0)
   ) -> (then)
   BackplateChangeOnUnhover!(
    execute=@n2.else
    self:WBP Sidebar Block Background Button Object Reference=Get Background()
    Button Type:E_UI_Sidebar_Button_Type Enum=Get ButtonType() -> (ButtonType=NewEnumerator0)
   ) -> (then)
  ]}
 OnHandleInputMethodChanged {nodes [
   n1 = OnHandleInputMethodChanged!() -> (then, NewInputType:ECommonInputType Enum) {is=title K2Node_FunctionEntry}
   > SetIsSelectable!(
    execute
    self:Common Button Base Object Reference
    bInIsSelectable:bool=Select!(MouseAndKeyboard:bool=false, Gamepad:bool=false, Touch:bool=true, Count:bool=false, Index:ECommonInputType Enum=@n1) -> (ReturnValue:bool)
   ) -> (then)
  ]}
 SetCompactUIStyle {nodes [
   SetCompactUIStyle() {is=title K2Node_FunctionEntry}
   > n1 = Set with Broadcast IsUsingCompactUI!(execute, IsUsingCompactUI:bool=IsUsingCompactUI()) -> (then, Output_Get:bool)
   > Branch(Condition=@n1)
   > SetMinDimensions!(execute, self:Common Button Base Object Reference, InMinWidth:int=80, InMinHeight:int=80) -> (then)
  ]}
]
variables [
 ButtonType {type=E_UI_Sidebar_Button_Type, default=Ghost}
 CornerRadius {type=number, default=10}
 CurrentInputType {type=ECommonInputType, default=MouseAndKeyboard}
 IsUsingCompactUI {type=boolean}
 RadiusInPixels {type=boolean, default=true}
]
widgets {
 info {
  inheritedWidgetCount=1
  namedSlotCount=1
  parentClass=/Script/FortniteUI.UIKitHoldableModularButton
  rootWidgetClass=&GridPanel
  widgetCount=5
 }
 widgets [
  Grid_ButtonRoot {widgetClassPath=&GridPanel, widget=Grid_ButtonRoot, properties {columnFill [2], rowFill [1]}}
  Background {
   widgetClassPath=/Sidebar/Buttons/Blocks/WBP_Sidebar_Block_Background_Button
   bIsVariable=true
   parent=Grid_ButtonRoot
   widget=Background
   slot {class=GridSlot}
  }
  Slot_ItemContent {widgetClassPath=&NamedSlot, bIsVariable=true, namedSlotHost=Background, widget=Slot_ItemContent}
  Slot_Controls {
   widgetClassPath=&NamedSlot
   bIsVariable=true
   parent=Grid_ButtonRoot
   widget=Slot_Controls
   slot {class=GridSlot, properties {column=1, layer=-99}}
  }
  InputActionWidget {widgetClassPath=&CommonActionWidget, bInherited=true, widget=null}
 ]
}
