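# Sidebar-styled text input field widget, derived from the UIKit input field base.
# PreConstruct applies compact-UI height, icon, font-size and corner-radius overrides, registers the background's focus outline as a modular block, and hides the base outline.
# It then sets backplate color, corners and padding, and rebuilds the text font at the chosen size.
# Hover, selection and enabled-state events drive ButtonBackground's backplate changes, with disabling also dimming the widget.
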WidgetBlueprint /Sidebar/Buttons/WBP_Sidebar_InputField
graphs [
 EventGraph {nodes [
   n1 = Event PreConstruct()
   > Parent: PreConstruct(IsDesignTime=@n1.IsDesignTime)
   > SetCompactUIStyle() {is=title K2Node_CallFunction}
   > RegisterModularBlock!(
    execute
    self:UIKit Modular Button Object Reference
    Block:UIKit Block Object Reference=Get FocusOutline!(self:WBP Sidebar Block Background Button Object Reference=Get ButtonBackground() -> (ButtonBackground:WBP Sidebar Block Background Button Object Reference)) -> (FocusOutline:WBP Sidebar Block Outline Object Reference)
   ) -> (then)
   > SetBlockVisibility!(
    execute
    self:WBP UIKit Block Base Object Reference=Get Block_Outline() -> (Block_Outline:WBP UIKit Block Outline Object Reference)
    Visible:bool=false
   ) -> (then)
   > InitializeButtonVisuals()
   n2 = Break Slate Font Info!(SlateFontInfo:"Slate Font Info Structure (by ref)"=Get Font() -> (Font:Slate Font Info Structure)) -> (
    FontObject:Object Reference
    FontMaterial:Object Reference
    OutlineSettings:Font Outline Settings Structure
    TypefaceFontName:Name
    Size:float=~
    LetterSpacing:int
    SkewAmount:float=~
    bForceMonospaced:bool=~
    bMaterialIsStencil:bool=~
    MonospacedWidth:float=~
   ) {is=title K2Node_BreakStruct}
   > SetFont!(
    execute
    self:WBP UIKit Block Editable Text Object Reference=Get Block_EditableText() -> (Block_EditableText:WBP UIKit Block Editable Text Object Reference)
    Font:Slate Font Info Structure=MakeSlateFontInfo!(
     FontObject:Object Reference=@n2
     FontMaterial:Object Reference=@n2.FontMaterial
     OutlineSettings:Font Outline Settings Structure=@n2.OutlineSettings
     TypefaceFontName:Name=@n2.TypefaceFontName
     Size:float=Get EditableText_FontSize() -> (EditableText_FontSize:float)
     LetterSpacing:int=@n2.LetterSpacing
     SkewAmount:float=@n2.SkewAmount
     bForceMonospaced:bool=@n2.bForceMonospaced
     bMaterialIsStencil:bool=@n2.bMaterialIsStencil
     MonospacedWidth:float=@n2.MonospacedWidth
    ) -> (ReturnValue:Slate Font Info Structure)
   ) -> (then)
   Event On Enabled() {is=title K2Node_Event}
   > Parent: On Enabled() {is=title K2Node_CallParentFunction}
   > SetRenderOpacity(self=Self-Reference(), InOpacity=1)
   > n3 = Branch(Condition=IsHovered())
   > BackplateChangeOnHover!(
    execute
    self:WBP Sidebar Block Background Button Object Reference=Get ButtonBackground()
    Button Type:E_UI_Sidebar_Button_Type Enum=NewEnumerator0
   ) -> (then)
   BackplateChangeOnUnhover!(
    execute=@n3.else
    self:WBP Sidebar Block Background Button Object Reference=Get ButtonBackground()
    Button Type:E_UI_Sidebar_Button_Type Enum=NewEnumerator0
   ) -> (then)
   Event On Disabled() {is=title K2Node_Event}
   > Parent: On Disabled() {is=title K2Node_CallParentFunction}
   > BackplateChangeOnDisabled!(
    execute
    self:WBP Sidebar Block Background Button Object Reference=Get ButtonBackground()
    Button Type:E_UI_Sidebar_Button_Type Enum=NewEnumerator0
   ) -> (then)
   > SetRenderOpacity(self=Self-Reference(), InOpacity=0.25)
   Event On Hovered()
   > Parent: On Hovered() {is=title K2Node_CallParentFunction}
   > BackplateChangeOnHover(self=Get ButtonBackground(), Button Type=NewEnumerator0)
   Event On Unhovered()
   > Parent: On Unhovered() {is=title K2Node_CallParentFunction}
   > BackplateChangeOnUnhover(self=Get ButtonBackground(), Button Type=NewEnumerator0)
   Event On Selected()
   > Parent: On Selected() {is=title K2Node_CallParentFunction}
   n4 = Get ButtonBackground()
   > Set IsSelected!(execute, IsSelected:bool=true, self:WBP Sidebar Block Background Button Object Reference=@n4) -> (then, Output_Get:bool)
   > BackplateChangeOnSelected!(execute, self:WBP Sidebar Block Background Button Object Reference=@n4, Button Type:E_UI_Sidebar_Button_Type Enum=NewEnumerator0) -> (then)
   Event On Deselected()
   > Parent: On Deselected() {is=title K2Node_CallParentFunction}
   n5 = Get ButtonBackground()
   > Set IsSelected(IsSelected=false, self=@n5)
   > BackplateChangeOnDeselected!(execute, self:WBP Sidebar Block Background Button Object Reference=@n5, Button Type:E_UI_Sidebar_Button_Type Enum=NewEnumerator0) -> (then)
  ]}
 InitializeButtonVisuals {nodes [
   InitializeButtonVisuals() {is=title K2Node_FunctionEntry}
   n1 = Get Backplate!(self:WBP Sidebar Block Background Button Object Reference=Get ButtonBackground()) -> (Backplate:WBP Sidebar Block Backplate Object Reference)
   > SetBackplateColor!(execute, self:WBP Sidebar Block Backplate Object Reference=@n1, Backplate Switch:E_UI_Sidebar_Button_Backplate_Color Enum=NewEnumerator1) -> (then)
   > SetVisibility(
    self=Get Backplate(self:WBP Sidebar Block Backplate Object Reference=@n1) -> (Backplate:Image Object Reference)
    InVisibility=SelfHitTestInvisible
   )
   > SetCornerRadius!(
    execute
    self:WBP Sidebar Block Background Button Object Reference=Get ButtonBackground()
    Radius In Pixels:bool=Get RadiusInPixels() -> (RadiusInPixels:bool)
    Corner Radius:double=Get CornerRadius() -> (CornerRadius:double)
   ) -> (then)
   > SetGridSlotPadding!(
    execute
    Widget:Widget Object Reference=Get InputActionWidget() -> (InputActionWidget:Common Action Widget Object Reference)
    Padding_Left:float=0
    Padding_Top:float=0
    Padding_Right:float=12
    Padding_Bottom:float=0
   ) -> (then)
  ]}
 SetCompactUIStyle {nodes [
   SetCompactUIStyle() {is=title K2Node_FunctionEntry}
   > Branch(Condition=IsUsingCompactUI())
   > SetSize!(
    execute
    self:WBP UIKit Input Field Base Object Reference
    WIdth:S UI Bool Float Structure
    Height_Enable_3_#1:bool=true
    Height_Value_4_#2:double=80
   ) -> (then)
   n1 = Get!() -> (
    IconProperties_Placement_67_#3:S UI Placement Structure
    IconProperties_ImageSize_72_#4:Vector 2D Structure
    IconProperties_Material_75_#5:Material Instance Object Reference
   )
   > SetIconProperties!(
    execute
    self:WBP UIKit Input Field Base Object Reference
    Data_Placement_67_#3:S UI Placement Structure=@n1
    Data_ImageSize_72_#4:Vector 2D Structure=(X=32,Y=32)
    Data_Material_75_#5:Material Instance Object Reference=@n1.IconProperties_Material_75_#5
   ) -> (then)
   > Set EditableText_FontSize(EditableText_FontSize:float=21) -> (Output_Get:float)
   > Set CornerRadius(CornerRadius:double=20) -> (Output_Get:double)
  ]}
]
variables [
 CornerRadius {type=number, default=15}
 EditableText_FontSize {type=number, default=15.75}
 RadiusInPixels {type=boolean, default=true}
]
widgets {
 info {
  inheritedWidgetCount=0
  namedSlotCount=2
  parentClass=/FortUILibrary/UIKit/InputField/WBP_UIKit_InputField.WBP_UIKit_InputField_C
  rootWidgetClass=null
  widgetCount=2
 }
 widgets [
  ButtonBackground {widgetClassPath=/Sidebar/Buttons/Blocks/WBP_Sidebar_Block_Background_Button, bIsVariable=true, widget=ButtonBackground}
  Slot_ItemContent {widgetClassPath=&NamedSlot, bIsVariable=true, namedSlotHost=ButtonBackground, widget=Slot_ItemContent}
 ]
}
